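# Sound cue for the locked-on tone of the LumpyRug launcher.
# A single looping wave player reuses the proximity grenade launcher's in-air beep loop, pitched up at the cue level.
# It plays in the SFX class with overridden attenuation and references a test-folder concurrency asset.
# The debug flag is left on.
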SoundCue /Game/Sounds/Fort_Weapons/Launchers/LumpyRug/Athena_LumpyRug_LockedOn_SoundCue
root=WavePlayer_1
nodes [{
  id=WavePlayer_1
  properties {bLooping=true, soundWaveAssetPtr=/Game/Sounds/Fort_Weapons/Launchers/Prox_GrenadeLauncher/Proximity_GrenadeLauncher_InAir_Beep_Loop}
 }]
properties {
 bDebug=true
 bOverrideAttenuation=true
 maxDistance=4000
 pitchMultiplier=1.75
 priority=2
 soundClassObject=&SFX
 volumeMultiplier=1
 concurrencyOverrides {resolutionRule=PreventNew}
 concurrencySet [/Game/Sounds/Fort_Test_Sounds/Tones/test_priority_conc]
}
